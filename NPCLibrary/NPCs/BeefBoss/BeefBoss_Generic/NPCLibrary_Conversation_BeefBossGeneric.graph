ConversationDatabase /NPCLibrary/NPCs/BeefBoss/BeefBoss_Generic/NPCLibrary_Conversation_BeefBossGeneric
entry_points [{tag=Conversation.Entry.BeefBossGeneric, destinations [#1]}]
nodes [
 ConversationEntryPointNode_7 {
  id=#1
  class=ConversationEntryPointNode
  outputs [#2]
  properties {entryTag {tagName=Conversation.Entry.BeefBossGeneric}}
 }
 Speech_C_4 {
  id=#2
  class=Speech_C
  outputs [#3, #4, #5, #6]
  properties {
   nodeName=""
   generalConfig {
    defaultMessage="Have a beefy day!"
    contextualMessages [
     {
      message=Look who it is....
      contextRequirements [{requirement {
         class=FortControllerRequirement_HasCID
         properties {allowedCIds [
           /BRCosmetics/Athena/Items/Cosmetics/Characters/CID_538_Athena_Commando_M_Taco
           /BRCosmetics/Athena/Items/Cosmetics/Characters/CID_797_Athena_Commando_F_Taco
           /BRCosmetics/Athena/Items/Cosmetics/Characters/CID_109_Athena_Commando_M_Pizza
           /BRCosmetics/Athena/Items/Cosmetics/Characters/CID_831_Athena_Commando_F_PIzzaPitMascot
          ]}
        }}]
      priority {curve {rowName=Tandem.Speech.Priority.OutfitReaction}}
     }
     {
      message="I was just flippin' some patties when some sort of... laser-sight showed up on my grill. Terrifying!"
      contextRequirements [{requirement {class=FortControllerRequirement_HasOwnedTag, properties {ownedTag {tagName=Athena.Quests.NPC.DialogueOverride.Nightmare}}}}]
      priority {curve {rowName=Tandem.Speech.Priority.HasQuestGameplayTag}}
     }
    ]
   }
  }
 }
 Speech_C_0 {
  id=#6
  class=Speech_C
  outputs [#7]
  properties {
   softTaskIcon=&T_UI_NPC_question
   generalConfig {defaultMessage=You know too much.}
   voiceOverTag {tagName=Pawn.Athena.NPC.Sound.Voice.Quest.Excited}
  }
  sub_nodes [ConversationChoiceNode_1 {class=ConversationChoiceNode, properties {defaultChoiceDisplayText=You look weird...}}]
 }
 BrawnySmasherNPC_C_0 {id=#7, class=DisguiseNPC_C, outputs []}
 ConversationLinkNode_0 {id=#5, class=ConversationLinkNode, properties {remoteEntryTag {tagName=Conversation.Entry.BasicServices}}}
 Speech_C_1 {
  id=#4
  class=Speech_C
  outputs [#8, #9]
  properties {
   softTaskIcon=/Game/Athena/HUD/Interaction/T_UI_NPC_Bounty
   generalConfig {defaultMessage="Ready to put the competition <Hot>on the grill?</>"}
  }
  sub_nodes [
   FortConversationRequirement_HasService_0 {class=FortConversationRequirement_HasService, properties {serviceTag {tagName=Tandem.Service.PlayerBounty}}}
   FortConversationChoiceWithPreviewNode_0 {class=FortConversationChoiceWithPreviewNode, properties {defaultChoiceDisplayText=BOUNTY}}
  ]
 }
 GrantPlayerBounty_C_0 {
  id=#9
  class=GrantPlayerBounty_C
  outputs []
  sub_nodes [
   ConversationChoiceNode_2 {class=ConversationChoiceNode, properties {defaultChoiceDisplayText=CONFIRM}}
   SideEffect_Client_PlayVOEvent_C_0 {class=SideEffect_Client_PlayVOEvent_C, properties {voiceOverTag {tagName=Pawn.Athena.NPC.Sound.Voice.Quest.Excited}}}
  ]
 }
 Back_C_0 {
  id=#8
  class=Back_C
  sub_nodes [
   ConversationChoiceNode_0 {class=ConversationChoiceNode, properties {defaultChoiceDisplayText=BACK}}
   SideEffect_Client_PlayVOEvent_C_1 {class=SideEffect_Client_PlayVOEvent_C, properties {voiceOverTag {tagName=Pawn.Athena.NPC.Sound.Voice.Quest.Disappointed}}}
  ]
 }
 ConversationLinkNode_1 {id=#3, class=ConversationLinkNode, properties {remoteEntryTag {tagName=Conversation.Entry.NPCLibrary.Quests}}}
]
